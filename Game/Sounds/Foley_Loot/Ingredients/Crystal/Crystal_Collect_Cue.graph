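# Sound cue for collecting a Crystal ingredient, routed through the Fort_Loot_Collect sound class.
# A Mixer plays two layered waves together, a real and a surreal crystal collect sound, with the surreal layer at half volume.
# The cue lowers overall volume, uses the shared prop-collect attenuation, and overrides concurrency to a small instance cap that stops the oldest.
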SoundCue /Game/Sounds/Foley_Loot/Ingredients/Crystal/Crystal_Collect_Cue
root=Mixer_2
nodes [
 {id=Mixer_2, children [WavePlayer_4,WavePlayer_5], properties {inputVolume [1,0.5]}}
 {id=WavePlayer_4, properties {soundWaveAssetPtr=&/Crystal_Collect_Real}}
 {id=WavePlayer_5, properties {soundWaveAssetPtr=&/Crystal_Collect_Surreal}}
]
properties {
 attenuationSettings=/Game/Sounds/Attenuation/Fort_Prop_Collect_Att
 bOverrideConcurrency=true
 maxDistance=2097152
 soundClassObject=&Fort_Loot_Collect
 volumeMultiplier=0.299
 concurrencyOverrides {maxCount=2, resolutionRule=StopOldest}
}
